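; Looping gameplay cue for Hermes PumpkinPass that toggles visibility of its spawned particles from the target's gameplay tags.
; On start it watches local controllers' view targets, so spectating also counts.
; When one views the cue target, it waits for ability system initialization, then runs paired async tag-query waits that hide particles on match and show them otherwise.
; Removal ends those waits and unbinds the delegates.
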(event EventOnLoopingStart(EitherVFX-preferthisversion)
  (params (MyTarget :type Actor%20Object%20Reference) (Parameters :type Gameplay%20Cue%20Parameters%20Structure%20%28by%20ref%29) (ParticleComponents :type Array%20of%20FXSystem%20Component%20Object%20References) (AudioComponents :type Array%20of%20Audio%20Component%20Object%20References))
  ;[0 0.15 0.01832 0.5] OnLoopingStart: When this gameplay cue is applied, for the local players, bind to know if their view target ever becomes the target of this GCN. We use view target so that it works while being a spectator as well.
  (|CallParentFunction MyTarget Parameters ParticleComponents AudioComponents)
  (bind _output_get (Variables|Transient|SetTargetActor MyTarget)) ; The target actor that the GCNL is applied to.
  (Variables|Transient|SetSpawnedParticleComponents ParticleComponents)
  (for _array_element (Fortnite|GetLocalFortPlayerControllers _output_get true true)
    (Utilities|IsValid _array_element
      (:Is%20Valid
        (EventDispatchers|BindEventtoOnViewTargetChanged _array_element (EventDispatchers|CreateEvent :!event Custom|HandleViewTargetOfLocalPlayerChanged))
        (CallFunction|HandleViewTargetOfLocalPlayerChanged _array_element 0 (Pawn|GetViewTarget _array_element))))))

(event EventOnRemoval(EitherVFX-preferthisversion)
  (params (MyTarget :type Actor%20Object%20Reference) (Parameters :type Gameplay%20Cue%20Parameters%20Structure%20%28by%20ref%29) (ParticleComponents :type Array%20of%20FXSystem%20Component%20Object%20References) (AudioComponents :type Array%20of%20Audio%20Component%20Object%20References) (BurstCameraShakeInstance :type Camera%20Shake%20Base%20Object%20Reference) (BurstDecalInstance :type Decal%20Actor%20Object%20Reference))
  ;[0.07292 0.03298 0 0.5] OnRemoval: Unbind delegates/cleanup.
  (|CallParentFunction MyTarget Parameters ParticleComponents AudioComponents BurstCameraShakeInstance BurstDecalInstance)
  (for _array_element (Fortnite|GetLocalFortPlayerControllers MyTarget true true)
    (Utilities|IsValid _array_element
      (:Is%20Valid
        (EventDispatchers|UnbindEventfromOnViewTargetChanged _array_element (EventDispatchers|CreateEvent :!event Custom|HandleViewTargetOfLocalPlayerChanged)))))
  (Ability|Async|EndAction (Variables|Transient|GetAsyncActions))
  (bind _asfort_player_pawn (Utilities|Casting|CastToFortPlayerPawn (Variables|Transient|GetTargetActor)))
  (EventDispatchers|UnbindEventfromOnAbilitySystemComponentInitialized _asfort_player_pawn (EventDispatchers|CreateEvent :!event Custom|HandleAbilitySystemComponentInit)))

(event Custom|HandleViewTargetOfLocalPlayerChanged
  (params (PlayerController :type Fort%20Player%20Controller%20Object%20Reference) (OldViewTarget :type Actor%20Object%20Reference) (NewViewTarget :type Actor%20Object%20Reference))
  ;[0.05046 0.000183 0.15 0.5] HandleViewTargetOfLocalPlayerChanged: If the view target of our local controllers become the target actor of this gameplay cue, then check to see if they have specific tags. Tags can only be checked once the ability system component is initizlied.
  (if (== NewViewTarget (Variables|Transient|GetTargetActor))
    (bind _asfort_player_pawn (Utilities|Casting|CastToFortPlayerPawn (Variables|Transient|GetTargetActor)))
    (EventDispatchers|BindEventtoOnAbilitySystemComponentInitialized _asfort_player_pawn (EventDispatchers|CreateEvent :!event Custom|HandleAbilitySystemComponentInit))
    (Utilities|IsValid (Ability|GetFortAbilitySystemComponent _asfort_player_pawn)
      (:Is%20Valid
        (CallFunction|HandleAbilitySystemComponentInit (Ability|GetFortAbilitySystemComponent _asfort_player_pawn) _asfort_player_pawn)))))

(event Custom|HandleAbilitySystemComponentInit
  (params (AbilitySystemComponent :type Fort%20Ability%20System%20Component%20Object%20Reference) (AffectedPawn :type Fort%20Player%20Pawn%20Object%20Reference))
  (bind _gameplaytagqueryhideparticle (Variables|Settings|GetGameplayTagQueryHideParticle))
  (bind _targetactor (Variables|Transient|GetTargetActor))
  ;[0 0.0438 0.15 0.5] HandleAbilitySystemComponentInit: Only bind to detect tag changes once the ability system component is initialized.
  (Ability|Async|EndAction (Variables|Transient|GetAsyncActions))
  (bind _asynctaskproxy (Ability|Tasks|WaitGameplayTagQueryOnActor _targetactor _gameplaytagqueryhideparticle)
    (:then
      (Utilities|Array|Add (Variables|Transient|GetAsyncActions) _asynctaskproxy)
      (bind _asynctaskproxy_2 (Ability|Tasks|WaitGameplayTagQueryOnActor _targetactor _gameplaytagqueryhideparticle "WhenFalse")
        (:then
          (Utilities|Array|Add (Variables|Transient|GetAsyncActions) _asynctaskproxy_2))
        (:Triggered
          (CallFunction|SetParticleComponentsEnabled true))))
    (:Triggered
      (CallFunction|SetParticleComponentsEnabled false))))

(event Custom|SetParticleComponentsEnabled
  (params (bEnabled :type Boolean))
  ;[0 0.03143 0.15 0.5] SetParticleComponentsEnabled: Set hidden or unhidden state of the particle components.
  (Development|SetHiddeninGame (Variables|Transient|GetSpawnedParticleComponents) (not bEnabled) false))

(fn ConstructionScript
  (|CallParentFunction))
